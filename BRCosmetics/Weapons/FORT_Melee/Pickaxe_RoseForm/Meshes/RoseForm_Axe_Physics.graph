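# Physics asset for the RoseForm two-handed pickaxe, defining capsule bodies for the handle and two dynamic bolt-socket bones.
# Each bolt socket is constrained directly to the handle, with very tight swing-1 and twist limit angles, so movement favors the swing-2 axis.
# The second socket explicitly frees swing 2, letting it rotate around that axis without a limit.
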PhysicsAsset /BRCosmetics/Weapons/FORT_Melee/Pickaxe_RoseForm/Meshes/RoseForm_Axe_Physics
bodies {
 Melee_TwoHanded_Handle [{shapeType=Capsule, length=18.8552, radius=2.2595}]
 dyn_bolt_socket_1 [{shapeType=Capsule, length=8.4458, radius=4, shapeName=dyn_bolt_socket_1_capsule, center=<0 0 -6.1793>, rotation=<rot 90 0 0>}]
 dyn_bolt_socket_2 [{shapeType=Capsule, length=5.9142, radius=2.9879, shapeName=dyn_bolt_socket_2_capsule, center=<0 0 4.4137>, rotation=<rot 90 0 0>}]
}
constraints [
 {bone1Name=dyn_bolt_socket_1, bone2Name=Melee_TwoHanded_Handle, swing1LimitDegrees=1, swing2LimitDegrees=45, twistLimitDegrees=1}
 {
  bone1Name=dyn_bolt_socket_2
  bone2Name=Melee_TwoHanded_Handle
  swing1LimitDegrees=1
  swing2LimitDegrees=45
  swing2Motion=Free
  twistLimitDegrees=1
 }
]
